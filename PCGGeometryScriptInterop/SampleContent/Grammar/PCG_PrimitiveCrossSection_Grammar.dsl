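; PCG sample graph that builds grammar-based buildings from tagged primitive actors overlapping the PCG volume.
; Primitive data from MeshBuildingTag actors goes through Primitive Cross-Section, producing splines with an extrusion vector attribute.
; Copy Attributes promotes ExtrudeVector to the data domain, since the cross-section node cannot create data attributes.
; The result feeds the PCG_BuildingSample subgraph as SplineInput, with exposed toggles for floor and module debugging and mesh spawning.
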(graph PCG_PrimitiveCrossSection_Grammar
  (params (DebugFloors :type Bool :default true) (DebugModules :type Bool :default true) (SpawnMeshes :type Bool :default true))
  (bind GetPrimitiveData_0 (Get%20Primitive%20Data :ActorSelector (struct :ActorFilter AllWorldActors :bMustOverlapSelf true :ActorSelectionTag MeshBuildingTag :bSelectMultiple true)))
  (bind PrimitiveCrossSection_1 (Primitive%20Cross-Section :ExtrusionVectorAttribute ExtrudeVector))
  (bind CopyAttributes_1 (Copy%20Attributes :InputSource ExtrudeVector :OutputTarget @Data.ExtrudeVector)) ; Primitive Cross-Section does not support\ncreating data attributes, so we need to \ncopy it there.
  (bind Subgraph_2 (Subgraph :Graph /PCG/SampleContent/Grammar/Graphs/PCG_BuildingSample.PCG_BuildingSample :DebugFloors DebugFloors :DebugModules DebugModules :SpawnMeshes SpawnMeshes))
  (connect PrimitiveCrossSection_1 Out CopyAttributes_1 Source)
  (connect PrimitiveCrossSection_1 Out CopyAttributes_1 Target)
  (connect GetPrimitiveData_0 Out PrimitiveCrossSection_1 In)
  (connect CopyAttributes_1 Out Subgraph_2 SplineInput))
